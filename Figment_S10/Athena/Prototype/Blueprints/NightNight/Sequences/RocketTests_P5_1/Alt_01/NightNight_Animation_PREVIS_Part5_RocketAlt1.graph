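# Previs Level Sequence for the NightNight sequence (Part 5, rocket alternate 1), animating a separated Geode rocket around a spawned scene origin, with an animated rail position.
# The rocket switches from blue engines to afterburner, carries camera shake, shrinks its main stage, then despawns.
# Four planes rotate while stretching enormously in Z, their materials fading in and flashing via Opacity and Multiplier.
# Two event triggers fire during the plane effect.
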LevelSequence /Figment_S10/Athena/Prototype/Blueprints/NightNight/Sequences/RocketTests_P5_1/Alt_01/NightNight_Animation_PREVIS_Part5_RocketAlt1
bindings [
 P_NightNight_Afterburner {
  id=#1
  parent_id=#2
  possessed_class=ParticleSystemComponent
  tracks [{
    type=MovieSceneParticleTrack
    display_name=Particle System
    sections [{
      type=MovieSceneParticleSection
      channels [None {type=MovieSceneScriptingParticleChannel, keys [[1023, "<ParticleKey.DEACTIVATE: 1>"], [1799, "<ParticleKey.ACTIVATE: 0>"]]}]
     }]
   }]
 }
 P_Geode_2_BlueEngines {
  id=#3
  parent_id=#2
  possessed_class=ParticleSystemComponent
  tracks [{
    type=MovieSceneParticleTrack
    display_name=Particle System
    sections [{
      type=MovieSceneParticleSection
      channels [None {type=MovieSceneScriptingParticleChannel, keys [[1023, "<ParticleKey.ACTIVATE: 0>"], [1799, "<ParticleKey.DEACTIVATE: 1>"]]}]
     }]
   }]
 }
 1x 1Plane on Z {
  id=#4
  possessed_class=FortStaticMeshActor
  tracks [
   {
    type=MovieScene3DTransformTrack
    sections [{
      type=MovieScene3DTransformSection
      channels [
       Location.X {type=MovieSceneScriptingDoubleChannel, default=97.9558, const [-98.0227,1]}
       Location.Y {type=MovieSceneScriptingDoubleChannel, default=-98.018, const [98.1248,1]}
       Location.Z {type=MovieSceneScriptingDoubleChannel, default=1530, keys [[531.0025,3021.0698],[1023.0025,1530]]}
       Rotation.X {type=MovieSceneScriptingDoubleChannel, default=0, const [0,2]}
       Rotation.Y {type=MovieSceneScriptingDoubleChannel, default=90, keys [[1503.0025,90],[1731,0]]}
       Rotation.Z {type=MovieSceneScriptingDoubleChannel, default=0, const [0,2]}
       Scale.X {type=MovieSceneScriptingDoubleChannel, default=1, const [1,1]}
       Scale.Y {type=MovieSceneScriptingDoubleChannel, default=1, const [1,1]}
       Scale.Z {type=MovieSceneScriptingDoubleChannel, default=9250000, keys [[531.0025,250000],[1503,0.001],[1563,9250000]]}
      ]
     }]
   }
   {
    type=MovieSceneSpawnTrack
    sections [{type=MovieSceneSpawnSection, channels [None {type=MovieSceneScriptingBoolChannel, default=false, keys [[0,false],[1503,true],[1946,false]]}]}]
   }
   {type=MovieScene3DAttachTrack, sections [{type=MovieScene3DAttachSection, end_frame=2425, start_frame=1023}]}
  ]
 }
 StaticMeshComponent0 {
  id=#5
  parent_id=#4
  possessed_class=StaticMeshComponent
  tracks [{
    type=MovieSceneComponentMaterialTrack
    display_name=Material Element 0
    sections [{
      type=MovieSceneParameterSection
      end_frame=2425
      start_frame=1023
      channels [
       Opacity {type=MovieSceneScriptingFloatChannel, keys [[1503.0025,0],[1731.0025,1]]}
       Multiplier {type=MovieSceneScriptingFloatChannel, default=10, keys [[1769,1],[1789,10],[1873,1]]}
      ]
     }]
   }]
 }
 StaticMeshComponent0 {
  id=#6
  parent_id=#7
  possessed_class=StaticMeshComponent
  tracks [{
    type=MovieSceneComponentMaterialTrack
    display_name=Material Element 0
    sections [{
      type=MovieSceneParameterSection
      end_frame=2425
      start_frame=1023
      channels [
       Opacity {type=MovieSceneScriptingFloatChannel, keys [[1503.005,0],[1731.005,1]]}
       Multiplier {type=MovieSceneScriptingFloatChannel, default=10, keys [[1769.0025,1],[1789.0025,10],[1873.0025,1]]}
      ]
     }]
   }]
 }
 Cube {
  id=#8
  possessed_class=FortStaticMeshActor
  tracks [
   {
    type=MovieScene3DTransformTrack
    sections [{
      type=MovieScene3DTransformSection
      channels [
       Location.X {type=MovieSceneScriptingDoubleChannel, default=0, const [-1024,1]}
       Location.Y {type=MovieSceneScriptingDoubleChannel, default=0, const [0,1]}
       Location.Z {type=MovieSceneScriptingDoubleChannel, default=0, const [0,1]}
       Rotation.X {type=MovieSceneScriptingDoubleChannel, default=0, const [0,1]}
       Rotation.Y {type=MovieSceneScriptingDoubleChannel, default=0, const [0,1]}
       Rotation.Z {type=MovieSceneScriptingDoubleChannel, default=0, const [0,1]}
       Scale.X {type=MovieSceneScriptingDoubleChannel, default=1, const [1,1]}
       Scale.Y {type=MovieSceneScriptingDoubleChannel, default=1, const [1,1]}
       Scale.Z {type=MovieSceneScriptingDoubleChannel, default=1, const [1,1]}
      ]
     }]
   }
   {
    type=MovieSceneSpawnTrack
    sections [{type=MovieSceneSpawnSection, channels [None {type=MovieSceneScriptingBoolChannel, default=false, keys [[1023,true],[2425,false]]}]}]
   }
   {type=MovieScene3DAttachTrack, sections [{type=MovieScene3DAttachSection, end_frame=3667, start_frame=1023}]}
  ]
 }
 Camera Rig Rail {
  id=#9
  possessed_class=CameraRig_Rail
  tracks [
   {type=MovieSceneSpawnTrack, sections [{type=MovieSceneSpawnSection, channels [None {type=MovieSceneScriptingBoolChannel, default=true}]}]}
   {
    type=MovieScene3DTransformTrack
    sections [{
      type=MovieScene3DTransformSection
      channels [
       Location.X {type=MovieSceneScriptingDoubleChannel, default=60.9126, const [60.9126,1]}
       Location.Y {type=MovieSceneScriptingDoubleChannel, default=0, const [0,1]}
       Location.Z {type=MovieSceneScriptingDoubleChannel, default=192179.8125, const [192179.8125,1]}
       Rotation.X {type=MovieSceneScriptingDoubleChannel, default=0, const [0,1]}
       Rotation.Y {type=MovieSceneScriptingDoubleChannel, default=0, const [0,1]}
       Rotation.Z {type=MovieSceneScriptingDoubleChannel, default=0, const [0,1]}
       Scale.X {type=MovieSceneScriptingDoubleChannel, default=1, const [1,1]}
       Scale.Y {type=MovieSceneScriptingDoubleChannel, default=1, const [1,1]}
       Scale.Z {type=MovieSceneScriptingDoubleChannel, default=1, const [1,1]}
      ]
     }]
   }
   {
    type=MovieSceneFloatTrack
    display_name=Current Position on Rail
    sections [{
      type=MovieSceneFloatSection
      channels [None {type=MovieSceneScriptingFloatChannel, default=0.3799, keys [[476,0],[1023,0.3799],[1795,0.6402],[2463,1]]}]
     }]
   }
   {type=MovieScene3DAttachTrack, sections [{type=MovieScene3DAttachSection, end_frame=3677, start_frame=1023}]}
  ]
 }
 SceneOrigin {
  id=#10
  possessed_class=FortStaticMeshActor
  tracks [
   {
    type=MovieScene3DTransformTrack
    sections [{
      type=MovieScene3DTransformSection
      channels [
       Location.X {type=MovieSceneScriptingDoubleChannel, default=65949.1797}
       Location.Y {type=MovieSceneScriptingDoubleChannel, default=-42629.9727}
       Location.Z {type=MovieSceneScriptingDoubleChannel, default=411558.9062}
       Rotation.X {type=MovieSceneScriptingDoubleChannel, default=0}
       Rotation.Y {type=MovieSceneScriptingDoubleChannel, default=0}
       Rotation.Z {type=MovieSceneScriptingDoubleChannel, default=0}
       Scale.X {type=MovieSceneScriptingDoubleChannel, default=1}
       Scale.Y {type=MovieSceneScriptingDoubleChannel, default=1}
       Scale.Z {type=MovieSceneScriptingDoubleChannel, default=1}
      ]
     }]
   }
   {type=MovieSceneSpawnTrack, sections [{type=MovieSceneSpawnSection, channels [None {type=MovieSceneScriptingBoolChannel, default=true}]}]}
  ]
 }
 Rocket_01_Sep {
  id=#2
  possessed_class=BP_Geode_NightNight_Separated_C
  tracks [
   {
    type=MovieSceneSpawnTrack
    sections [{
      type=MovieSceneSpawnSection
      channels [None {type=MovieSceneScriptingBoolChannel, default=false, keys [[0,false],[1023,true],[2095.8,false]]}]
     }]
   }
   {
    type=MovieScene3DTransformTrack
    sections [{
      type=MovieScene3DTransformSection
      end_frame=3455
      start_frame=1023
      channels [
       Location.X {type=MovieSceneScriptingDoubleChannel, default=0}
       Location.Y {type=MovieSceneScriptingDoubleChannel, default=0}
       Location.Z {type=MovieSceneScriptingDoubleChannel, default=0}
       Rotation.X {type=MovieSceneScriptingDoubleChannel, default=-90}
       Rotation.Y {type=MovieSceneScriptingDoubleChannel, default=0}
       Rotation.Z {type=MovieSceneScriptingDoubleChannel, default=90}
      ]
     }]
   }
   {type=MovieScene3DAttachTrack, sections [{type=MovieScene3DAttachSection, end_frame=4487, start_frame=1023}]}
  ]
 }
 StaticMeshComponent0 {
  id=#11
  parent_id=#12
  possessed_class=StaticMeshComponent
  tracks [{
    type=MovieSceneComponentMaterialTrack
    display_name=Material Element 0
    sections [{
      type=MovieSceneParameterSection
      end_frame=2425
      start_frame=1022
      channels [
       Opacity {type=MovieSceneScriptingFloatChannel, keys [[1503.0025,0],[1731.0025,1]]}
       Multiplier {type=MovieSceneScriptingFloatChannel, default=10, keys [[1769.0025,1],[1789.0025,10],[1873.0025,1]]}
      ]
     }]
   }]
 }
 NightNight_SmallerRocket_CameraShakeComponent {
  id=#13
  parent_id=#2
  possessed_class=NightNight_SmallerRocket_CameraShakeComponent_C
  tracks [{
    type=MovieSceneFloatTrack
    display_name=Shake Strength
    sections [{type=MovieSceneFloatSection, channels [None {type=MovieSceneScriptingFloatChannel, default=0.5, const [0.5,1]}]}]
   }]
 }
 1x 1Plane on Z {
  id=#12
  possessed_class=FortStaticMeshActor
  tracks [
   {
    type=MovieScene3DTransformTrack
    sections [{
      type=MovieScene3DTransformSection
      channels [
       Location.X {type=MovieSceneScriptingDoubleChannel, default=97.9558, const [-98.0105,1]}
       Location.Y {type=MovieSceneScriptingDoubleChannel, default=-98.018, const [-97.9172,1]}
       Location.Z {type=MovieSceneScriptingDoubleChannel, default=1530, keys [[771.0025,3021.0181],[1023.0025,1530]]}
       Rotation.X {type=MovieSceneScriptingDoubleChannel, default=-90, keys [[1503,-90],[1731.0025,0]]}
       Rotation.Y {type=MovieSceneScriptingDoubleChannel, default=0, const [0,2]}
       Rotation.Z {type=MovieSceneScriptingDoubleChannel, default=0, const [0,2]}
       Scale.X {type=MovieSceneScriptingDoubleChannel, default=1, const [1,1]}
       Scale.Y {type=MovieSceneScriptingDoubleChannel, default=1, const [1,1]}
       Scale.Z {type=MovieSceneScriptingDoubleChannel, default=9250000, keys [[771.0025,250000],[1503.0025,0.001],[1563.0025,9250000]]}
      ]
     }]
   }
   {
    type=MovieSceneSpawnTrack
    sections [{
      type=MovieSceneSpawnSection
      channels [None {type=MovieSceneScriptingBoolChannel, default=false, keys [[360.0025,false],[1503.0025,true],[1946,false]]}]
     }]
   }
   {type=MovieScene3DAttachTrack, sections [{type=MovieScene3DAttachSection, end_frame=2425, start_frame=1023}]}
  ]
 }
 1x 1Plane on Z {
  id=#14
  possessed_class=FortStaticMeshActor
  tracks [
   {
    type=MovieScene3DTransformTrack
    sections [{
      type=MovieScene3DTransformSection
      channels [
       Location.X {type=MovieSceneScriptingDoubleChannel, default=97.9558, const [97.9558,1]}
       Location.Y {type=MovieSceneScriptingDoubleChannel, default=-98.018, const [-98.018,1]}
       Location.Z {type=MovieSceneScriptingDoubleChannel, default=1530, keys [[531.0025,3020.9512],[1023.0025,1530]]}
       Rotation.X {type=MovieSceneScriptingDoubleChannel, default=0, const [0,2]}
       Rotation.Y {type=MovieSceneScriptingDoubleChannel, default=0, keys [[1503.0025,-90],[1731.0025,0]]}
       Rotation.Z {type=MovieSceneScriptingDoubleChannel, default=0, const [0,2]}
       Scale.X {type=MovieSceneScriptingDoubleChannel, default=1, const [1,1]}
       Scale.Y {type=MovieSceneScriptingDoubleChannel, default=1, const [1,1]}
       Scale.Z {type=MovieSceneScriptingDoubleChannel, default=0.001, keys [[531.0025,250000],[1503.005,0.001],[1563.005,9250000]]}
      ]
     }]
   }
   {
    type=MovieSceneSpawnTrack
    sections [{
      type=MovieSceneSpawnSection
      channels [None {type=MovieSceneScriptingBoolChannel, default=false, keys [[0,false],[1503.0025,true],[1946,false]]}]
     }]
   }
   {type=MovieScene3DAttachTrack, sections [{type=MovieScene3DAttachSection, end_frame=2425, start_frame=1023}]}
  ]
 }
 1x 1Plane on Z {
  id=#7
  possessed_class=FortStaticMeshActor
  tracks [
   {
    type=MovieScene3DTransformTrack
    sections [{
      type=MovieScene3DTransformSection
      channels [
       Location.X {type=MovieSceneScriptingDoubleChannel, default=97.9558, const [97.9819,1]}
       Location.Y {type=MovieSceneScriptingDoubleChannel, default=-98.018, const [98.1384,1]}
       Location.Z {type=MovieSceneScriptingDoubleChannel, default=1530, keys [[531.0025,3021.0732],[1023.0025,1530]]}
       Rotation.X {type=MovieSceneScriptingDoubleChannel, default=90, keys [[1503.0025,90],[1731,0]]}
       Rotation.Y {type=MovieSceneScriptingDoubleChannel, default=0, const [0,2]}
       Rotation.Z {type=MovieSceneScriptingDoubleChannel, default=0, const [0,2]}
       Scale.X {type=MovieSceneScriptingDoubleChannel, default=1, const [1,1]}
       Scale.Y {type=MovieSceneScriptingDoubleChannel, default=1, const [1,1]}
       Scale.Z {type=MovieSceneScriptingDoubleChannel, default=9250000, keys [[531.0025,250000],[1503,0.001],[1563,9250000]]}
      ]
     }]
   }
   {
    type=MovieSceneSpawnTrack
    sections [{
      type=MovieSceneSpawnSection
      channels [None {type=MovieSceneScriptingBoolChannel, default=false, keys [[0,false],[1503.005,true],[1946,false]]}]
     }]
   }
   {type=MovieScene3DAttachTrack, sections [{type=MovieScene3DAttachSection, end_frame=2425, start_frame=1023}]}
  ]
 }
 MissileSilo_Stage2_Main {
  id=#15
  parent_id=#2
  possessed_class=StaticMeshComponent
  tracks [{
    type=MovieScene3DTransformTrack
    sections [{
      type=MovieScene3DTransformSection
      channels [
       Scale.X {type=MovieSceneScriptingDoubleChannel, default=0, keys [[1909,1],[1946,0]]}
       Scale.Y {type=MovieSceneScriptingDoubleChannel, default=0, keys [[1909,1],[1946,0]]}
       Scale.Z {type=MovieSceneScriptingDoubleChannel, default=0, keys [[1909,1],[1946,0]]}
      ]
     }]
   }]
 }
 StaticMeshComponent0 {
  id=#16
  parent_id=#14
  possessed_class=StaticMeshComponent
  tracks [{
    type=MovieSceneComponentMaterialTrack
    display_name=Material Element 0
    sections [{
      type=MovieSceneParameterSection
      end_frame=2425
      start_frame=1022
      channels [
       Opacity {type=MovieSceneScriptingFloatChannel, keys [[1503.0025,0],[1731.0025,1]]}
       Multiplier {type=MovieSceneScriptingFloatChannel, default=10, keys [[1769.0025,1],[1789.0025,10],[1873.0025,1]]}
      ]
     }]
   }]
 }
]
display_rate {denominator=1, numerator=60}
tick_resolution {denominator=1, numerator=24000}
tracks [{
  type=MovieSceneEventTrack
  sections [{
    type=MovieSceneEventTriggerSection
    channels [None {
      type=MovieSceneScriptingEventChannel
      keys [[1503, "<Struct 'MovieSceneEvent' (0x49e50fe7bc0) {}>"], [1851.0025, "<Struct 'MovieSceneEvent' (0x49e50fe7b20) {}>"]]
     }]
   }]
 }]
