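# Level Sequence for an alternate Part 5 animation previs in the NightNight prototype, filed under rocket tests.
# It has no bindings of its own and layers eight overlapping subsequence sections, mostly sharing one range after a longer-leading first section.
# Two fade tracks each add a double fade pulse, around where the first and second subsections end.
# An event trigger track is also present, though its target is not shown.
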LevelSequence /Figment_S10/Athena/Prototype/Blueprints/NightNight/Sequences/RocketTests_P5_1/Alt_01/NightNight_Animation_PREVIS_Part5_alt
bindings []
display_rate {denominator=1, numerator=60}
tick_resolution {denominator=1, numerator=24000}
tracks [
 {
  type=MovieSceneSubTrack
  display_name=Subsequences
  sections [
   {type=MovieSceneSubSection, end_frame=2623, start_frame=450}
   {type=MovieSceneSubSection, end_frame=3836, start_frame=1672}
   {type=MovieSceneSubSection, end_frame=4326, start_frame=1672}
   {type=MovieSceneSubSection, end_frame=4326, start_frame=1672}
   {type=MovieSceneSubSection, end_frame=4326, start_frame=1672}
   {type=MovieSceneSubSection, end_frame=4326, start_frame=1672}
   {type=MovieSceneSubSection, end_frame=4326, start_frame=1672}
   {type=MovieSceneSubSection, end_frame=4326, start_frame=1672}
  ]
 }
 {
  type=MovieSceneFadeTrack
  sections [{
    type=MovieSceneFadeSection
    channels [None {type=MovieSceneScriptingFloatChannel, default=0, keys [[3671,0],[3689.61,1],[3702.92,0],[3779.955,0],[3802.0275,1],[3843.6425,1],[3950,0]]}]
   }]
 }
 {
  type=MovieSceneFadeTrack
  sections [{
    type=MovieSceneFadeSection
    channels [None {type=MovieSceneScriptingFloatChannel, default=0, keys [[2539,0],[2548,1],[2578,0],[2608,0],[2668,1],[2699,1],[2818,0]]}]
   }]
 }
 {
  type=MovieSceneEventTrack
  sections [{
    type=MovieSceneEventTriggerSection
    channels [None {type=MovieSceneScriptingEventChannel, const ["<Struct 'MovieSceneEvent' (0x49e50fe78a0) {}>", 1]}]
   }]
 }
]
